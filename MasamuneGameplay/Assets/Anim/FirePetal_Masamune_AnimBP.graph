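# Player item-layer AnimBlueprint for the FirePetal Masamune melee weapon, overriding the full-body and final-pose layers.
# Full body layers a non-looping charged-slash animation from Spine_01 while charging, otherwise hover idle or incoming pose, then adds pelvis yaw for melee twist.
# Final pose picks a root-pitched DefaultSlot for air montages or the FullBody slot, then modifies the BothHandFK curve.
# Other layers and EventGraph are unwired; ItemPreIK holds disconnected bone transforms.
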AnimBlueprint /MasamuneGameplay/Assets/Anim/FirePetal_Masamune_AnimBP
target_skeleton=&Fortnite_M_Avg_Player_Skeleton
graphs [
 AnimGraph {nodes [Output Pose()]}
 EventGraph {links [], nodes []}
 ItemAimOffset {nodes [
   Input Pose!() -> (Pose, AimOffsetAlpha:double, Yaw:double, Pitch:double) {name=InPose_AimOffset}
   Input Pose!() -> (Pose) {name=InPose_PreAimOffsets}
   Output Pose() {layerGroup=Item Override}
  ]}
 ItemFallAdditive {nodes [
   Input Pose() {name=InPoseItemFallAdditive}
   Output Pose() {layerGroup=Item Override}
  ]}
 ItemFallLandAdditive {nodes [
   Input Pose() {name=InPoseItemFallLandAdditive}
   Output Pose() {layerGroup=Item Override}
  ]}
 ItemFallLevel2Override {nodes [
   Input Pose() {name=InPose_FallLevel2}
   Output Pose() {layerGroup=Item Override}
  ]}
 ItemFinalPoseOverride {nodes [
   n1 = Use cached pose 'FinalPoseCache'()
   n2 = "Transform (Modify) Bone - Bone: Root"!(
    ComponentPose=Local To Component(LocalPose=Slot 'DefaultSlot'(Source=@n1))
    Alpha=Get PitchAlpha() -> (PitchAlpha:float)
    Translation="0,0,0"
    Rotation_Roll:float=Get ReturnValueYPitch() -> (ReturnValueYPitch:float)
    Rotation_Pitch:float=0
    Rotation_Yaw:float=0
    Scale="1,1,1"
   ) -> (Pose) {rotationMode=BMM_Additive, rotationSpace=BCS_BoneSpace, alphaBoolBlend {blendOption=Linear}, boneToModify {boneName=Root}}
   n3 = Component To Local(ComponentPose=@n2)
   n4 = Slot 'FullBody'(Source=Use cached pose 'FinalPoseCache'()) {slotName=FullBody}
   n5 = Get bPlayAirMontage() -> (bPlayAirMontage:bool)
   Output Pose(Result=Modify Curve(
     SourcePose=Blend Poses by bool(BlendPose_0=@n3, BlendPose_1=@n4, bActiveValue=@n5) {blendType=Linear}
     CurveValues_0=1
     Alpha=Get BothHandFK() -> (BothHandFK:float)
    ) {curveNames [BothHandFK], curveValues [0]}) {layerGroup=Item Override}
   Save cached pose 'FinalPoseCache'(Pose=Input Pose() {name=InFinalPose}) {cacheName=FinalPoseCache}
  ]}
 ItemFullBodyAimOffsets {nodes [
   Input Pose!() -> (Pose, FullBody_AimOffsetAlpha:double, Aim_Yaw:double, Aim_Pitch:double) {name=InPoseFullBodyAimOffsets}
   Input Pose!() -> (Pose) {name=InPosePreFullBodyAimOffsets}
   Output Pose() {layerGroup=Item Override}
  ]}
 ItemFullBodyOverride {nodes [
   n1 = Blend Poses by bool(
    BlendPose_0=Sequence Player 'TyphoonBlade_HoverIdle'() {sequence=&/TyphoonBlade_HoverIdle}
    BlendPose_1=Use cached pose 'PostMontageCache'()
    bActiveValue=Get bIsHovering() -> (bIsHovering:bool)
   ) {blendType=Linear}
   n2 = Layered blend per bone(
    BasePose=Use cached pose 'PostMontageCache'()
    BlendPoses_0=Sequence Player 'Masamune_ChargedSlash_Loop'() {bLoopAnimation=false, sequence=/MasamuneGameplay/Assets/Anim/ChargeSlash/Masamune_ChargedSlash_Loop}
   ) {bMeshSpaceRotationBlend=true, layerSetup [{branchFilters [{blendDepth=5, boneName=Spine_01}]}]}
   n3 = Get bCharging() -> (bCharging:bool)
   n4 = "Transform (Modify) Bone - Bone: pelvis"!(
    ComponentPose=Local To Component(LocalPose=Blend Poses by bool(BlendPose_0=@n2, BlendPose_1=@n1, BlendTime_0=0.2, BlendTime_1=0.2, bActiveValue=@n3) {blendType=Linear})
    Alpha=Get MeleeTwistAlpha() -> (MeleeTwistAlpha:double)
    Rotation_Roll:float=0
    Rotation_Pitch:float=0
    Rotation_Yaw:float=Get MeleeTwist() -> (MeleeTwist:float)
   ) -> (Pose) {
    rotationMode=BMM_Additive
    rotationSpace=BCS_WorldSpace
    translationSpace=BCS_WorldSpace
    alphaBoolBlend {blendOption=Linear}
    boneToModify {boneName=pelvis}
   }
   Output Pose(Result=Component To Local(ComponentPose=@n4)) {layerGroup=Item Override}
   Save cached pose 'PostMontageCache'(Pose=Input Pose() {name=InPoseFullBody}) {cacheName=PostMontageCache}
  ]}
 ItemIdleOverride {nodes [
   Input Pose() {name=InPose_Idle}
   Output Pose() {layerGroup=Item Override}
  ]}
 ItemJetPackJumpAdditive {nodes [Output Pose() {layerGroup=Item Override}]}
 ItemJetPackStartAdditive {nodes [Output Pose() {layerGroup=Item Override}]}
 ItemJumpLoopAdditive {nodes [
   Input Pose() {name=InPoseItemJumpLoopAdditive}
   Output Pose() {layerGroup=Item Override}
  ]}
 ItemJumpOverride {nodes [
   Input Pose() {name=InPoseJump}
   Output Pose() {layerGroup=Item Override}
  ]}
 ItemJumpUpAdditive {nodes [
   Input Pose() {name=InPoseItemJumpUpAdditive}
   Output Pose() {layerGroup=Item Override}
  ]}
 ItemPreAimOffsets {nodes [
   Input Pose() {name=InPosePreAimOffsets}
   Output Pose() {layerGroup=Item Override}
  ]}
 ItemPreIK {nodes [
   n1 = "Transform (Modify) Bone - Bone: pelvis"!(ComponentPose, Alpha=1, Translation:Vector="0,0,0", Rotation_Roll=-50, Rotation_Pitch=0, Rotation_Yaw=0, Scale:Vector="1,1,1") -> (Pose) {rotationMode=BMM_Additive, rotationSpace=BCS_BoneSpace, alphaBoolBlend {blendOption=Linear}, boneToModify {boneName=pelvis}}
   "Transform (Modify) Bone - Bone: Spine_01"!(ComponentPose=@n1, Alpha=1, Translation="0,0,0", Rotation_Roll:float=50, Rotation_Pitch:float=0, Rotation_Yaw:float=0, Scale="1,1,1") -> (Pose) {rotationMode=BMM_Additive, rotationSpace=BCS_BoneSpace, alphaBoolBlend {blendOption=Linear}, boneToModify {boneName=Spine_01}}
   Output Pose() {layerGroup=Item Override}
   Save cached pose 'PreIkCache'(Pose=Input Pose() {name=InPose_PreIK}) {cacheName=PreIkCache}
   Use cached pose 'PreIkCache'()
   "To Float (Boolean)"(InBool=Get bCharging())
  ]}
 ItemRightHandIK {nodes [
   Input Pose() {name=InPosePreRightHandIK}
   Input Pose() {name=InPosePostRightHandIK}
   Output Pose() {layerGroup=Item Override}
  ]}
 ItemSkeletalControl {nodes [
   Input Pose()
   Output Pose() {layerGroup=Item Override}
  ]}
 ItemSwimJumpFallLoop {nodes [Output Pose() {layerGroup=Item Override}]}
 ItemSwimJumpSurfaceEnd {nodes [Output Pose() {layerGroup=Item Override}]}
 ItemSwimJumpSurfaceLoop {nodes [Output Pose() {layerGroup=Item Override}]}
 ItemUpperBody {nodes [
   Input Pose!() -> (Pose, InputParam:Fort Anim Input Adjusted Aim Structure) {name=InPoseUpperBody}
   Output Pose() {layerGroup=Item Override}
  ]}
 ItemZipLineStartAdditive {nodes [Output Pose() {layerGroup=Item Override}]}
]
variables []
