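# Sound cue for the Masamune charge attack slash, playing a single charge swing wave with melee weapon sound class and Masamune attack attenuation.
# A root modulator adds slight random variation, then branches on the islocal and isenemy parameters pick a fixed-volume modulator.
# Local and enemy swings, or those with either parameter unset, play at full volume, while non-local, non-enemy swings play at half volume.
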SoundCue /MasamuneGameplay/Assets/Audio/SC_Masamune_Attack_Charge_Slash_01
root=Modulator_0
nodes [
 {id=Branch_2, children [Modulator_2,Branch_3,Modulator_2], properties {boolParameterName=islocal}}
 {id=Branch_3, children [Modulator_4,Modulator_3,Modulator_4], properties {boolParameterName=isenemy}}
 {id=Modulator_0, children [Branch_2], properties {pitchMin=1}}
 {id=Modulator_2, children [WavePlayer_0], properties {pitchMax=1, pitchMin=1, volumeMax=1, volumeMin=1}}
 {id=Modulator_3, children [WavePlayer_0], properties {pitchMax=1, pitchMin=1, volumeMax=0.5, volumeMin=0.5}}
 {id=Modulator_4, children [WavePlayer_0], properties {pitchMax=1, pitchMin=1, volumeMax=1, volumeMin=1}}
 {id=WavePlayer_0, properties {soundWaveAssetPtr=&/Masamune_Attack_Charge_Swing_01}}
]
properties {
 attenuationSettings=&/Att_Masamune_Attacks
 maxDistance=5500
 soundClassObject=&Fort_WeaponMelee_PSM
 volumeMultiplier=1.589
}
