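# Sound cue for the Masamune dive slash attack, playing one of three equally weighted launch sounds through a shared default modulator.
# Branches on islocal and isenemy play local and enemy sounds at full volume and non-local, non-enemy ones at half volume; unset parameters take the full-volume path.
# It uses Masamune attack attenuation and the melee weapon sound class, with per-owner concurrency limited to two, stopping the oldest.
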SoundCue /MasamuneGameplay/Assets/Audio/SC_Masamune_Attack_Dive_Slash_02
root=Branch_2
nodes [
 {id=Branch_2, children [Modulator_0,Branch_3,Modulator_0], properties {boolParameterName=islocal}}
 {id=Branch_3, children [Modulator_4,Modulator_3,Modulator_4], properties {boolParameterName=isenemy}}
 {id=Modulator_0, children [Modulator_2], properties {pitchMax=1, pitchMin=1, volumeMax=1, volumeMin=1}}
 {id=Modulator_2, children [Random_1]}
 {id=Modulator_3, children [Modulator_2], properties {pitchMax=1, pitchMin=1, volumeMax=0.5, volumeMin=0.5}}
 {id=Modulator_4, children [Modulator_2], properties {pitchMax=1, pitchMin=1, volumeMax=1, volumeMin=1}}
 {id=Random_1, children [WavePlayer_4,WavePlayer_5,WavePlayer_6], properties {weights [1,1,1]}}
 {id=WavePlayer_4, properties {soundWaveAssetPtr=&/Masamune_Attack_Dive_Slash_Launch_01}}
 {id=WavePlayer_5, properties {soundWaveAssetPtr=&/Masamune_Attack_Dive_Slash_Launch_02}}
 {id=WavePlayer_6, properties {soundWaveAssetPtr=&/Masamune_Attack_Dive_Slash_Launch_03}}
]
properties {
 attenuationSettings=&/Att_Masamune_Attacks
 bOverrideConcurrency=true
 maxDistance=5500
 soundClassObject=&Fort_WeaponMelee_PSM
 volumeMultiplier=1.589
 concurrencyOverrides {bLimitToOwner=true, maxCount=2, resolutionRule=StopOldest}
}
