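# Sound cue for the Masamune's normal melee attack swing, played as a spatialized player melee weapon sound with attenuation and concurrency limits.
# Four equally weighted random swing waves feed a shared default Modulator for slight variation.
# Nested islocal and isenemy branches reuse that chain through separate volume Modulators.
# Non-local, non-enemy swings play at half volume, while local, enemy and unset cases keep full volume.
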SoundCue /MasamuneGameplay/Assets/Audio/SC_Masamune_Attack_Swing_Normal_01
root=Branch_2
nodes [
 {id=Branch_2, children [Modulator_2,Branch_3,Modulator_2], properties {boolParameterName=islocal}}
 {id=Branch_3, children [Modulator_4,Modulator_3,Modulator_4], properties {boolParameterName=isenemy}}
 {id=Modulator_0, children [Random_0]}
 {id=Modulator_2, children [Modulator_0], properties {pitchMax=1, pitchMin=1, volumeMax=1, volumeMin=1}}
 {id=Modulator_3, children [Modulator_0], properties {pitchMax=1, pitchMin=1, volumeMax=0.5, volumeMin=0.5}}
 {id=Modulator_4, children [Modulator_0], properties {pitchMax=1, pitchMin=1, volumeMax=1, volumeMin=1}}
 {id=Random_0, children [WavePlayer_0,WavePlayer_1,WavePlayer_2,WavePlayer_3], properties {weights [1,1,1,1]}}
 {id=WavePlayer_0, properties {soundWaveAssetPtr=&/Masamune_Attack_Swing_Norm_01_01}}
 {id=WavePlayer_1, properties {soundWaveAssetPtr=&/Masamune_Attack_Swing_Norm_01_02}}
 {id=WavePlayer_2, properties {soundWaveAssetPtr=&/Masamune_Attack_Swing_Norm_01_03}}
 {id=WavePlayer_3, properties {soundWaveAssetPtr=&/Masamune_Attack_Swing_Norm_01_04}}
]
properties {
 attenuationSettings=&/Att_Masamune_Attacks
 maxDistance=5500
 soundClassObject=&Fort_WeaponMelee_PSM
 volumeMultiplier=1.589
 attenuationOverrides {distanceAlgorithm=NaturalSound, falloffDistance=2000, attenuationShapeExtents {x=500}}
 concurrencySet [&Player_Melee_Weapon_Conc]
}
