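# Sound cue for the Masamune item drop, routed through the Fort_Loot_Collect sound class.
# A single wave player plays Masamune_Drop_01 through a modulator at the cue root.
# Spatialization uses the shared Fort_Weapon_Drop_Att attenuation, and a cue-level concurrency override prevents new instances once the limit is reached.
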SoundCue /MasamuneGameplay/Assets/Audio/SC_Masamune_Drop
root=Modulator_0
nodes [{id=Modulator_0, children [WavePlayer_1]}, {id=WavePlayer_1, properties {soundWaveAssetPtr=&/Masamune_Drop_01}}]
properties {
 attenuationSettings=&Fort_Weapon_Drop_Att
 bOverrideConcurrency=true
 maxDistance=3000
 soundClassObject=&Fort_Loot_Collect
 volumeMultiplier=1
 concurrencyOverrides {resolutionRule=PreventNew}
}
