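# SoundCue for the Masamune weapon's drop landing sound.
# A Modulator node feeds a single wave player that plays Masamune_Drop_Land_01.
# The cue uses the shared Fortnite weapon drop attenuation and the melee weapon sound class.
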SoundCue /MasamuneGameplay/Assets/Audio/SC_Masamune_Drop_Land
root=Modulator_0
nodes [
 {id=Modulator_0, children [WavePlayer_21]}
 {id=WavePlayer_21, properties {soundWaveAssetPtr=&/Masamune_Drop_Land_01}}
]
properties {
 attenuationSettings=&Fort_Weapon_Drop_Att
 maxDistance=3000
 soundClassObject=&Fort_WeaponMelee
 volumeMultiplier=1
}
