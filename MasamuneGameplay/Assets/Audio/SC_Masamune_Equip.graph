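# Sound cue for the Masamune equip sound, classed under the Fort_WeaponMelee sound class.
# A Random node with equal weights picks one of three equip wave variations and feeds it through a root Modulator left at default settings.
# The SA_Masamune_Equip attenuation asset handles spatial falloff.
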SoundCue /MasamuneGameplay/Assets/Audio/SC_Masamune_Equip
root=Modulator_0
nodes [
 {id=Modulator_0, children [Random_0]}
 {id=Random_0, children [WavePlayer_0,WavePlayer_1,WavePlayer_2], properties {weights [1,1,1]}}
 {id=WavePlayer_0, properties {soundWaveAssetPtr=&/Masamune_Equip_01}}
 {id=WavePlayer_1, properties {soundWaveAssetPtr=&/Masamune_Equip_02}}
 {id=WavePlayer_2, properties {soundWaveAssetPtr=&/Masamune_Equip_03}}
]
properties {
 attenuationSettings=&/SA_Masamune_Equip
 maxDistance=2800
 soundClassObject=&Fort_WeaponMelee
 volumeMultiplier=1
}
