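# Sound cue for the Masamune weapon's charged impact, played as attenuated 3D audio.
# A random node picks one of four equally weighted charged-impact waves and feeds it through a root modulator.
# It uses the Masamune impacts attenuation, a melee weapon sound class and a boosted volume.
# Playback is limited by a player melee weapon concurrency set.
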SoundCue /MasamuneGameplay/Assets/Audio/SC_Masamune_Impact_Charged_Att
root=Modulator_0
nodes [
 {id=Modulator_0, children [Random_1]}
 {id=Random_1, children [WavePlayer_0,WavePlayer_5,WavePlayer_6,WavePlayer_7], properties {weights [1,1,1,1]}}
 {id=WavePlayer_0, properties {soundWaveAssetPtr=&/Masamune_Impact_Charged_01}}
 {id=WavePlayer_5, properties {soundWaveAssetPtr=&/Masamune_Impact_Charged_02}}
 {id=WavePlayer_6, properties {soundWaveAssetPtr=&/Masamune_Impact_Charged_03}}
 {id=WavePlayer_7, properties {soundWaveAssetPtr=&/Masamune_Impact_Charged_04}}
]
properties {
 attenuationSettings=/MasamuneGameplay/Assets/Audio/Settings/AS_Masamune_Impacts
 maxDistance=2500
 soundClassObject=&Fort_WeaponMelee_PSM
 volumeMultiplier=1.6
 concurrencyOverrides {resolutionRule=PreventNew}
 concurrencySet [&Player_Melee_Weapon_Conc]
}
