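# Sound cue for the Masamune sprint loop, playing a looping movement wind wave.
# An Enveloper ramps volume in over about 0.75 seconds, then feeds a default random Modulator.
# Branches on the islocal and isenemy parameters select fixed volume Modulators: local, enemy or unset play at full volume, non-enemy others at half.
# Masamune attack attenuation and a speed-boost concurrency set apply, with an owner-limited single-instance override also specified.
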SoundCue /MasamuneGameplay/Assets/Audio/SC_Masamune_Sprint_Loop
root=Branch_2
nodes [
 {id=Branch_2, children [Modulator_2,Branch_3,Modulator_2], properties {boolParameterName=islocal}}
 {id=Branch_3, children [Modulator_4,Modulator_3,Modulator_4], properties {boolParameterName=isenemy}}
 {
  id=Enveloper_0
  children [WavePlayer_0]
  properties {pitchCurve {editorCurveData {keys [{time=0}]}}, volumeCurve {editorCurveData {keys [<at 0 0>,<at 0.0018 0.1538>,<at 0.75 1>]}}}
 }
 {id=Modulator_0, children [Enveloper_0]}
 {id=Modulator_2, children [Modulator_0], properties {pitchMax=1, pitchMin=1, volumeMax=1, volumeMin=1}}
 {id=Modulator_3, children [Modulator_0], properties {pitchMax=1, pitchMin=1, volumeMax=0.5, volumeMin=0.5}}
 {id=Modulator_4, children [Modulator_0], properties {pitchMax=1, pitchMin=1, volumeMax=1, volumeMin=1}}
 {id=WavePlayer_0, properties {bLooping=true, soundWaveAssetPtr=&/Masamune_Movement_Wind_Loop_02}}
]
properties {
 attenuationSettings=&/Att_Masamune_Attacks
 maxDistance=5500
 priority=0.5
 soundClassObject=&Fort_WeaponMelee_PSM
 volumeMultiplier=1
 concurrencyOverrides {bLimitToOwner=true, maxCount=1}
 concurrencySet [/Game/Sounds/Creative/Concurrency/Creative_SpeedBoost_Increase_Conc]
}
