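# Unlit, translucent, two-sided slash effect material for Niagara sprite and mesh particles.
# An HDR color parameter, cyan-blue by default, multiplies particle color and feeds the MF_FX_EmissiveAdjustments function for glow.
# Opacity multiplies particle alpha by the red channel of T_Slash2, then passes through a DepthFade with a parameterized distance to soften intersections with scene geometry.
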Material /MasamuneGameplay/Assets/FX/M_Slash2
graph {
 properties {
  bUsedWithNiagaraMeshParticles=true
  bUsedWithNiagaraSprites=true
  blendMode=BLEND_Translucent
  pixelDepthOffsetMode=PDOM_Legacy
  shadingModel=MSM_Unlit
  twoSided=true
  usedShadingModels=Unlit
 }
 nodes [
  n1 = ParticleColor()
  DepthFade(
   Opacity=Multiply(A=@n1.A, B=TextureSample() {texture=&/T_Slash2}.R)
   FadeDistance=ScalarParameter() {defaultValue=50, parameterName=DF}
  )
  MaterialFunctionCall(Emissive=Multiply(A=VectorParameter() {defaultValue=<rgba 2 14.6667 40 1>}.RGB, B=@n1.RGB)) {materialFunction=&MF_FX_EmissiveAdjustments}
 ]
}
