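# Unlit, two-sided translucent prototype material for Masamune's charge reticle on Niagara sprites, drawn after motion blur with depth testing disabled.
# A double-sided Fresnel blends two similar greens, which particle color tints before a shared emissive-adjustment function.
# The same Fresnel multiplies the alpha of a spinning reticle texture and an Opacity parameter to form a faint translucency mask.
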Material /MasamuneGameplay/Assets/M_Proto_Masamune_ChargeRet
graph {
 properties {
  bAutomaticallySetUsageInEditor=false
  bDisableDepthTest=true
  bUseTranslucencyVertexFog=false
  bUsedWithNiagaraSprites=true
  blendMode=BLEND_Translucent
  pixelDepthOffsetMode=PDOM_Legacy
  shadingModel=MSM_Unlit
  translucencyPass=MTP_AfterMotionBlur
  twoSided=true
  usedShadingModels=Unlit
 }
 nodes [
  n1 = Constant3Vector() {constant=<rgba 0.3854 0.8438 0.2187 1>}
  n2 = Constant3Vector() {constant=<rgba 0.4044 0.8438 0.1334 1>}
  n3 = MaterialFunctionCall(Brightness=Constant() {r=2}, Exponent=Constant() {r=4}) {materialFunction=&MF_Fresnel_FX}
  MaterialFunctionCall(Emissive=Multiply(A=LinearInterpolate(A=@n1, B=@n2, Alpha=@n3.Fresnel Double Sided), B=ParticleColor().RGB)) {materialFunction=&MF_FX_EmissiveAdjustments}
  n4 = Rotator() {speed=3}
  Multiply(
   A=Multiply(
    A=@n3.Fresnel Double Sided
    B=TextureSample(UVs=@n4) {samplerType=SAMPLERTYPE_LinearColor, texture=&/T_Proto_Masamune_ChargeRet}.A
   )
   B=ScalarParameter() {defaultValue=0.2, parameterName=Opacity}
  )
 ]
}
