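# Anim Blueprint for the TyphoonBlade sharp skeleton that adds procedural motion to the blade's three-bone dynamic ribbon.
# It copies the pose from another mesh, applies DefaultSlot, then alternates Modify Bone and spring-constrained AnimDynamics nodes down the ribbon bones in component space.
# Each Modify Bone adds a per-bone ClothRotator rotation in world space and feeds in a shared Scale variable.
# The event graph is empty, so the rotators are not driven here.
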AnimBlueprint /MasamuneGameplay/Assets/Meshes/TyphoonBlade_Sharp_Skeleton_AnimBP
target_skeleton=&/TyphoonBlade_Sharp_Skeleton
graphs [
 AnimGraph {nodes [
   n1 = Copy Pose From Mesh()
   n2 = "Transform (Modify) Bone - Bone: dyn_ribbon_MSTR"(
    ComponentPose=Local To Component(LocalPose=Slot 'DefaultSlot'(Source=@n1))
    Alpha=1
    Translation="0,0,0"
    Rotation=Get ClothRotator_01() -> (ClothRotator_01:Rotator)
    Scale=Get Scale() -> (Scale:Vector)
   ) {
    rotationMode=BMM_Additive
    rotationSpace=BCS_WorldSpace
    translationSpace=BCS_BoneSpace
    alphaBoolBlend {blendOption=Linear}
    alphaScaleBiasClamp {bInterpResult=true, interpSpeedDecreasing=2, interpSpeedIncreasing=2}
    boneToModify {boneName=dyn_ribbon_MSTR}
   }
   n3 = Anim Dynamics - Bone: dyn_ribbon_01(ComponentPose=@n2, Alpha=1, ExternalForce="0,0,0") {
    angularDampingOverride=1
    angularSpringConstant=5000
    bAngularSpring=true
    bOverrideAngularDamping=true
    bUseGravityOverride=true
    gravityScale=0.35
    simulationSpace=RootRelative
    alphaBoolBlend {blendOption=Linear}
    boundBone {boneName=dyn_ribbon_01}
    componentLinearAccScale=<1 1 1>
    componentLinearVelScale=<1 1 1>
    gravityOverride {z=-980}
    physicsBodyDefinitions [{
      boundBone {boneName=dyn_ribbon_01}
      boxExtents=<25 25 25>
      constraintSetup {angularLimitsMax=<10 60 60>, angularLimitsMin=<-10 -60 -60>, angularTarget=<1 0 0>}
      localJointOffset=<4 0 0>
     }]
   }
   n4 = "Transform (Modify) Bone - Bone: dyn_ribbon_02"(ComponentPose=@n3, Alpha=1, Translation="0,0,0", Rotation=Get ClothRotator_02() -> (ClothRotator_02:Rotator), Scale=Get Scale()) {
    rotationMode=BMM_Additive
    rotationSpace=BCS_WorldSpace
    translationSpace=BCS_BoneSpace
    alphaBoolBlend {blendOption=Linear}
    alphaScaleBiasClamp {bInterpResult=true, interpSpeedDecreasing=2, interpSpeedIncreasing=2}
    boneToModify {boneName=dyn_ribbon_02}
   }
   n5 = Anim Dynamics - Bone: dyn_ribbon_02(ComponentPose=@n4, Alpha=1, ExternalForce="0,0,0") {
    angularDampingOverride=1
    angularSpringConstant=5000
    bAngularSpring=true
    bOverrideAngularDamping=true
    bUseGravityOverride=true
    gravityScale=0.35
    simulationSpace=RootRelative
    alphaBoolBlend {blendOption=Linear}
    boundBone {boneName=dyn_ribbon_02}
    componentLinearAccScale=<1 1 1>
    componentLinearVelScale=<1 1 1>
    gravityOverride {z=-980}
    physicsBodyDefinitions [{
      boundBone {boneName=dyn_ribbon_02}
      boxExtents=<25 25 25>
      constraintSetup {angularLimitsMax=<10 60 60>, angularLimitsMin=<-10 -60 -60>, angularTarget=<1 0 0>}
      localJointOffset=<4 0 0>
     }]
   }
   n6 = "Transform (Modify) Bone - Bone: dyn_ribbon_03"(ComponentPose=@n5, Alpha=1, Translation="0,0,0", Rotation=Get ClothRotator_03() -> (ClothRotator_03:Rotator), Scale=Get Scale()) {
    rotationMode=BMM_Additive
    rotationSpace=BCS_WorldSpace
    translationSpace=BCS_BoneSpace
    alphaBoolBlend {blendOption=Linear}
    alphaScaleBiasClamp {bInterpResult=true, interpSpeedDecreasing=2, interpSpeedIncreasing=2}
    boneToModify {boneName=dyn_ribbon_03}
   }
   n7 = Anim Dynamics - Bone: dyn_ribbon_03(ComponentPose=@n6, Alpha=1, ExternalForce="0,0,0") {
    angularDampingOverride=1
    angularSpringConstant=5000
    bAngularSpring=true
    bOverrideAngularDamping=true
    bUseGravityOverride=true
    gravityScale=0.35
    simulationSpace=RootRelative
    alphaBoolBlend {blendOption=Linear}
    boundBone {boneName=dyn_ribbon_03}
    componentLinearAccScale=<1 1 1>
    componentLinearVelScale=<1 1 1>
    gravityOverride {z=-980}
    physicsBodyDefinitions [{
      boundBone {boneName=dyn_ribbon_03}
      boxExtents=<25 25 25>
      constraintSetup {angularLimitsMax=<10 60 60>, angularLimitsMin=<-10 -60 -60>, angularTarget=<1 0 0>}
      localJointOffset=<4 0 0>
     }]
   }
   Output Pose(Result=Component To Local(ComponentPose=@n7))
  ]}
 EventGraph {links [], nodes []}
]
variables [Scale {type=Vector, default=<0.4 0.4 0.4>}]
